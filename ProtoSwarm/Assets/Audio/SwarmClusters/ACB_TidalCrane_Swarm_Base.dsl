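; Audio cluster behavior base for TidalCrane swarms, playing one MetaSound that follows the cluster's position in place of per-actor handling.
; Start spawns the sound and stop fades it out.
; Size changes leaving the cached range select matching ambient loop settings, then send a random wave, change trigger and master volume to the MetaSound.
; Member actors receive loose gameplay tags while in the cluster, removed on leave or stop.
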(event AudioClusterBehavior|EventOnStart
  (params (InitialSize :type Integer))
  (Variables|Swarm|Transient|SetCurrent%20Size InitialSize)
  (Variables|Swarm|Transient|GetSwarm%20Audio%20Component
    (:then
      _)
    (:else
      (bind _returnvalue (Audio|SpawnSoundatLocation (Variables|Swarm|GetSwarm%20MetaSound) (Variables|Swarm|Transient|GetCurrent%20Position)))
      (Variables|Swarm|Transient|SetSwarm%20Audio%20Component _returnvalue)))
  (Swarm|UpdateMetasoundParameters))

(event AudioClusterBehavior|EventOnStop
  (Utilities|FlowControl|Sequence
    (:then_0
      (bind _swarm_audio_component (Variables|Swarm|Transient|GetSwarm%20Audio%20Component)
        (:then
          (Audio|Components|Audio|FadeOut _swarm_audio_component 0.1 0.0 "Linear" :!class /Script/Engine.AudioComponent))
        (:else)))
    (:then_1
      (for _array_element (Variables|Swarm|Transient|GetActors)
        (CallFunction|RemoveActor _array_element)))))

(event AudioClusterBehavior|EventOnSizeUpdated
  (params (NewSize :type Integer))
  (bind _output_get (Variables|Swarm|Transient|SetCurrent%20Size NewSize))
  (if (Swarm|IsNotInCurrentRange _output_get)
    (Swarm|UpdateMetasoundParameters)))

(event AudioClusterBehavior|EventOnPositionUpdated
  (params (NewPosition :type Vector%20%28by%20ref%29 :default "0, 0, 0"))
  (Variables|Swarm|Transient|SetCurrent%20Position NewPosition)
  (bind _swarm_audio_component (Variables|Swarm|Transient|GetSwarm%20Audio%20Component)
    (:then
      (Transformation|SetWorldLocation _swarm_audio_component (Variables|Swarm|Transient|GetCurrent%20Position) false false))
    (:else)))

(event AudioClusterBehavior|EventOnActorAdded
  (params (Actor :type Actor%20Object%20Reference))
  (Utilities|Array|AddUnique (Variables|Swarm|Transient|GetActors) Actor)
  (Ability|GameplayEffect|AddLooseGameplayTags Actor (Variables|Swarm|GetTags%20to%20Add)))

(event AudioClusterBehavior|EventOnActorRemoved
  (params (Actor :type Actor%20Object%20Reference))
  (CallFunction|RemoveActor Actor))

(event Custom|Remove%20Actor
  (params (Actor :type Actor%20Object%20Reference))
  (Utilities|Array|RemoveItem (Variables|Swarm|Transient|GetActors) Actor)
  (Ability|GameplayEffect|RemoveLooseGameplayTags Actor (Variables|Swarm|GetTags%20to%20Add)))

(fn IsNotinCurrentRange
  (params (Size :type Integer))
  (bind (_min _max) (Utilities|Struct|BreakInt32Interval (Variables|Swarm|Transient|GetCurrent%20Cluster%20Size%20Range)))
  (return (not (Math|Integer|InRange(Integer) Size _min _max))))

(fn UpdateMetasoundParameters
  (bind _current_settings (Variables|Swarm|Transient|GetCurrent%20Settings))
  (bind (_loops_3_022dee064cfc0bfbdbac4881b227fe87 _volume_17_7a179ea8482ab4f156af24a91aa9e772) (Utilities|Struct|BreakTidalCraneMovementSwarmAudioSettings _current_settings))
  (Variables|Swarm|Transient|GetSwarm%20Audio%20Component
    (:then
      (bind _array_element (Utilities|Array|ForEachLoopwithBreak (Variables|Swarm|GetAmbient%20Swarm%20Loops))
        (:LoopBody
          (bind _output_get (LocalVariables|SetSettings _array_element))
          (if (Math|Integer|InRange(Integer) (Variables|Swarm|Transient|GetCurrent%20Size) (Utilities|Struct|BreakInt32Interval (Utilities|Struct|BreakTidalCraneMovementSwarmAudioSettings _output_get)) (.Max (Utilities|Struct|BreakInt32Interval (Utilities|Struct|BreakTidalCraneMovementSwarmAudioSettings _output_get))))
            (Variables|Swarm|Transient|SetCurrent%20Settings (LocalVariables|GetSettings))))
        (:Completed
          _))
      (Utilities|FlowControl|Sequence
        (:then_0
          (Variables|Swarm|Transient|SetCurrent%20Cluster%20Size%20Range (Utilities|Struct|BreakTidalCraneMovementSwarmAudioSettings (Variables|Swarm|Transient|GetCurrent%20Settings))))
        (:then_1
          (Utilities|FlowControl|Sequence
            (:then_0
              (Audio|Parameter|SetObjectParameter (Variables|Swarm|Transient|GetSwarm%20Audio%20Component) "Swarm.Movement.Wave" (Utilities|Array|RandomArrayItem _loops_3_022dee064cfc0bfbdbac4881b227fe87))
              (Audio|Parameter|ExecuteTriggerParameter (Variables|Swarm|Transient|GetSwarm%20Audio%20Component) "Swarm.Movement.WaveChanged")
              (Audio|Parameter|SetFloatParameter (Variables|Swarm|Transient|GetSwarm%20Audio%20Component) "MasterVolume" _volume_17_7a179ea8482ab4f156af24a91aa9e772 :!class /Script/AudioExtensions.AudioParameterControllerInterface))))))
    (:else)))
