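# Looping stereo MetaSound source for TidalCrane swarm cluster movement audio, with shared concurrency and cluster attenuation settings.
# A loop-swapper patch switches to the supplied wave on the WaveChanged trigger, which also rerolls a small random pitch shift.
# Listener distance is smoothed, then mapped to a fade from inner radius across falloff distance.
# That gain, multiplied by MasterVolume, scales both channels, giving graph-driven distance attenuation with runtime-adjustable radii.
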MetaSoundSource /ProtoSwarm/Assets/Audio/SwarmClusters/MSS_TidalCrane_Movement_Swarm
inputs [
 MasterVolume {type=Float}
 Swarm.Movement.Attenuation.FalloffDistance {type=Float}
 Swarm.Movement.Attenuation.InnerRadius {type=Float}
 Swarm.Movement.Wave {type=WaveAsset}
 Swarm.Movement.WaveChanged {type=Trigger}
 UE.Attenuation.Distance {type=Float}
 UE.Source.OnPlay {type=Trigger}
]
interfaces [UE.Attenuation:1.0, UE.OutputFormat.Stereo:1.0, UE.Source:1.1]
outputs [
 "UE.OutputFormat.Stereo.Audio:0" {type=Audio}
 "UE.OutputFormat.Stereo.Audio:1" {type=Audio}
]
pages [Default {
  input_defaults {
   MasterVolume=1
   Swarm.Movement.Attenuation.FalloffDistance=0
   Swarm.Movement.Attenuation.InnerRadius=1200
   Swarm.Movement.Wave=null
   Swarm.Movement.WaveChanged=null
   UE.Attenuation.Distance=0
   UE.Source.OnPlay=null
  }
  nodes [
   n1 = &/MSP_LoopSwapperStereo(
    NextSoundToPlay=^Swarm.Movement.Wave
    PlayNextSound=^Swarm.Movement.WaveChanged
    PitchShift=UE.RandomFloat(Next=^Swarm.Movement.WaveChanged) {also {defaults {Max=2, Min=-2}}}.Value
   )
   n2 = UE.Multiply:Float(
    AdditionalOperands=^MasterVolume
    PrimaryOperand=MapRange.MapRange:Float(
     In Range A=^Swarm.Movement.Attenuation.InnerRadius
     In Range B=UE.Add:Float(AdditionalOperands=^Swarm.Movement.Attenuation.FalloffDistance, PrimaryOperand=^Swarm.Movement.Attenuation.InnerRadius).Out
     In=UE.InterpTo:Audio(Target=^UE.Attenuation.Distance) {also {defaults {Interp Time=0.5}}}.Value
    ) {also {defaults {Out Range A=1, Out Range B=0}}}.Out Value
   )
   Multiply_0 = UE.Multiply:Audio by Float(PrimaryOperand=@n1.Out R, AdditionalOperands=@n2.Out)
   Multiply_1 = UE.Multiply:Audio by Float(PrimaryOperand=@n1.Out L, AdditionalOperands=@n2.Out)
  ]
  links [{dst="@UE.OutputFormat.Stereo.Audio:1", src=Multiply_0.Out}, {dst="@UE.OutputFormat.Stereo.Audio:0", src=Multiply_1.Out}]
 }]
properties {
 assetImportData=&self:AssetImportData
 attenuationSettings=/ProtoSwarm/Assets/Audio/TidalCrane_Swarmer/ATT_TidalCrane_Movement_Swarmer_Cluster
 bLooping=true
 duration=10000
 maxDistance=9000
 numChannels=2
 outputFormat=Stereo
 sampleRate=48000
 soundClassObject=/Game/Sounds/FortSoundClassesAndModes/Fort_Creatures_Steps_Normal
 concurrencySet [/ProtoSwarm/Assets/Audio/Settings/CONC_TidalCrane_Swarmer_Movement_Shared]
}
